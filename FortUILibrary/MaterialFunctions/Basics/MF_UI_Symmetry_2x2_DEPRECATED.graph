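# Deprecated UI material function that mirrors UVs horizontally and vertically into a 2x2 symmetric layout.
# It doubles and fracs the input UV, then lerps each axis between the offset coordinate and its flipped counterpart using a saturated floor mask.
# The mirrored UVs pass through CustomRotator and ScaleUVsByCenter, driven by Rotation and Scale inputs.
# The right and bottom half masks are also output alongside the final UVs.
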MaterialFunction /FortUILibrary/MaterialFunctions/Basics/MF_UI_Symmetry_2x2_DEPRECATED
graph {
 properties {description=Mirrors visually the input horizontally and vertically.}
 nodes [
  n1 = Multiply(
   A=FunctionInput(Preview=TextureCoordinate()) {bUsePreviewValueAsDefault=true, description=Input the UV, inputName=UV, inputType=FunctionInput_Vector2}
   B=Constant2Vector() {g=2, r=2}
  )
  n2 = FunctionInput(Preview=Constant2Vector()) {bUsePreviewValueAsDefault=true, description=Defines an offset, inputName=Offset, inputType=FunctionInput_Vector2, sortPriority=2}
  n3 = Add(A=Frac(None=@n1), B=@n2)
  n4 = MaterialFunctionCall(Float2=@n3) {materialFunction=&BreakOutFloat2Components}
  n5 = MaterialFunctionCall(Float2=Add(A=OneMinus(None=@n3), B=Multiply(A=@n2) {constB=2})) {materialFunction=&BreakOutFloat2Components}
  n6 = MaterialFunctionCall(Float2=Saturate(None=Floor(None=@n1))) {materialFunction=&BreakOutFloat2Components}
  n7 = AppendVector(A=LinearInterpolate(A=@n4.R, B=@n5.R, Alpha=@n6.R), B=LinearInterpolate(A=@n4.G, B=@n5.G, Alpha=@n6.G))
  n8 = FunctionInput(Preview=Constant()) {bUsePreviewValueAsDefault=true, description=Defines a rotation value, inputName=Rotation, inputType=FunctionInput_Scalar, sortPriority=4}
  n9 = FunctionInput(Preview=Constant() {r=0.5}) {bUsePreviewValueAsDefault=true, description=Defines a scale factor, inputName=Scale, inputType=FunctionInput_Vector2, sortPriority=5}
  FunctionOutput(None=MaterialFunctionCall(
    UVs=MaterialFunctionCall(UVs=@n7, "Rotation Angle (0-1)"=@n8) {materialFunction=&CustomRotator}.Rotated Values
    Texture Scale=@n9
   ) {materialFunction=&ScaleUVsByCenter}.UVs) {description=Outputs the symmetry}
  FunctionOutput(None=@n6.R) {description=Outputs the right side mask of the symmetry, outputName=Right Side Mask, sortPriority=1}
  FunctionOutput(None=@n6.G) {description=Outputs the bottom side mask of the symmetry, outputName=Bottom Side Mask, sortPriority=2}
 ]
}
