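# Deprecated WIP UI material function that fakes a perspective 3D card by remapping quad UVs onto a rotated plane.
# A Custom HLSL node casts a view ray per UV from FOV, intersects a plane at Offset, and undoes the rotation.
# It outputs transformed UV, normal, view direction, depth and an edge-clipping mask that Tile disables per axis.
# Angle is scaled by 2π, so it acts as turns despite its radians label.
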MaterialFunction /FortUILibrary/MaterialFunctions/Basics/MF_UI_WIP_3DCard_DEPRECATED
graph {
 properties {previewBlendMode=BLEND_Translucent, libraryCategoriesText [""]}
 nodes [
  MaterialExpressionFunctionInput_0 = FunctionInput(Preview=TextureCoordinate()) {bUsePreviewValueAsDefault=true, description=quad location, inputName=UV, inputType=FunctionInput_Vector2}
  MaterialExpressionMultiply_0 = Multiply(A=FunctionInput() {bUsePreviewValueAsDefault=true, description="x,y,z rotation angles (Radians)", inputName=Angle, sortPriority=2}) {constB=6.2832}
  n1 = FunctionInput() {bUsePreviewValueAsDefault=true, inputName=Offset, inputType=FunctionInput_Scalar, sortPriority=3, previewValue=<2.4142 0 0 1>}
  MaterialExpressionMax_0 = Max(A=@n1)
  n2 = FunctionInput() {
   bUsePreviewValueAsDefault=true
   description=field of view angle in degrees
   inputName=FOV
   inputType=FunctionInput_Scalar
   sortPriority=4
   previewValue=<45 0 0 1>
  }
  MaterialExpressionTangent_0 = Tangent!(None=Multiply(A=@n2) {constB=0.5}) {period=360}
  MaterialExpressionSaturate_0 = Saturate(None=FunctionInput() {bUsePreviewValueAsDefault=true, inputName=Tile, inputType=FunctionInput_Vector2, sortPriority=1})
  n3 = Custom!(
   uv=@MaterialExpressionFunctionInput_0
   angle=@MaterialExpressionMultiply_0
   Offset=@MaterialExpressionMax_0
   FOV_scale=@MaterialExpressionTangent_0
   tile=@MaterialExpressionSaturate_0
  ) {
   code="\r\nfloat3x3 Xrot, Yrot, Zrot;\r\n\r\nXrot = float3x3(1, 0 , 0, \r\n\t\t\t\t0, cos(angle.x), sin(angle.x), \r\n\t\t\t\t0, -sin(angle.x), cos(angle.x));\r\n\r\nYrot = float3x3(cos(angle.y), 0 , -sin(angle.y), \r\n\t\t\t\t0, 1, 0,\r\n\t\t\t\tsin(angle.y), 0, cos(angle.y));\r\n\t\t\t\t\r\nZrot = float3x3(cos(angle.z), sin(angle.z), 0, \r\n\t\t\t\t-sin(angle.z), cos(angle.z), 0,\r\n\t\t\t\t0, 0, 1);\r\n\t\t\t\t\r\nfloat3x3 rot = mul(mul(Yrot , Xrot), Zrot);\r\n\r\n//create rotated plane\r\nfloat3 norm = mul(rot, float3(0,0,1)).xyz;\r\n\r\n//project uv onto plane\r\nfloat3 viewRay = normalize(float3(FOV_scale*(2*uv.xy-1),1));\r\n\r\nfloat d = dot(viewRay.xyz ,norm);\r\nfloat3 pv = float3(0,0,Offset);\r\nfloat t = dot(pv,norm)/d;\r\n\r\n//undo rotation to get new UV position\r\nfloat3 wpos = viewRay *t;\r\nfloat3 nuv = mul(transpose(rot), wpos - float3(0,0,Offset) );\r\n\r\nfloat alpha = 1;\r\nfloat a0,a1;\r\na0 = lerp(1,0, saturate(floor(abs(nuv.x))));\r\na0 = lerp(a0, 1, floor(tile.x));\r\na1 = lerp(1,0, saturate(floor(abs(nuv.y))));\r\na1 = lerp(a1, 1, floor(tile.y));\r\n\r\nalpha = a0*a1;\r\n\r\nnuv.xy = .5*nuv.xy + .5;\r\n\r\nMask = alpha;\r\n\r\n//output normal\r\nNormal.xyz = norm;\r\n\r\n//output viewDir\r\nViewDir = normalize(viewRay);\r\n\r\n//output Depth\r\nDepth = t;\r\n\r\n//output UV\r\nreturn nuv.xy;\r\n\t\t\t\t\r\n\t\t\t\t"
   description=Rotate
   outputType=CMOT_Float2
   additionalOutputs [
    {outputName=Mask, outputType=CMOT_Float1}
    {outputName=Normal, outputType=CMOT_Float4}
    {outputName=ViewDir, outputType=CMOT_Float3}
    {outputName=Depth, outputType=CMOT_Float1}
   ]
   inputs [
    {inputName=uv, input {expression=&self:MaterialExpressionFunctionInput_0}}
    {inputName=angle, input {expression=&self:MaterialExpressionMultiply_0}}
    {inputName=Offset, input {expression=&self:MaterialExpressionMax_0}}
    {inputName=FOV_scale, input {expression=&self:MaterialExpressionTangent_0}}
    {inputName=tile, input {expression=&self:MaterialExpressionSaturate_0}}
   ]
  }
  FunctionOutput(None=@n3.return) {description="Transformed UV ", outputName=UV}
  FunctionOutput(None=@n3.Normal) {description="Normalized surface normal vector ", outputName=Normal, sortPriority=2}
  FunctionOutput(None=@n3.ViewDir) {outputName=ViewDir, sortPriority=3}
  FunctionOutput(None=@n3.Depth) {description=linear distance from camera to surface point, outputName=Depth, sortPriority=4}
  FunctionOutput(None=@n3.Mask) {description="Clipping mask ", outputName=Mask, sortPriority=1}
 ]
}
