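# UI material function that returns aspect-ratio corrected 9-slice UVs, keeping corners proportioned while the middle stretches with the widget's display area.
# It derives per-axis scale from the pixel width/height ratios and applies it inward from both edges of the clamped input UVs, which default to Normalized UV.
# Each edge's result is capped at the 0.5 midpoint, and a floored UV×2 selects the half without branching.
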MaterialFunction /FortUILibrary/MaterialFunctions/Basics/MF_UI_9SliceUVs
graph {
 properties {
  bExposeToLibrary=true
  description=Returns aspect-ratio corrected UVs that scale and maintain corners according to their display area.
  previewMaterialDomain=MD_UI
  libraryCategoriesText [UI_Library]
 }
 nodes [
  n1 = MaterialFunctionCall() {materialFunction=&GetUserInterfaceUV}
  n2 = ComponentMask(None=@n1.Pixel Size) {r=true}
  n3 = ComponentMask(None=@n1.Pixel Size) {g=true}
  n4 = Max(A=AppendVector(A=Divide(A=@n2, B=@n3), B=Divide(A=@n3, B=@n2)))
  n5 = FunctionInput(Preview=@n1.Normalized UV) {
   bUsePreviewValueAsDefault=true
   description="Optional custom UV mapping (defaults to Normalized UVs)."
   inputName=UVs
   inputType=FunctionInput_Vector2
  }
  n6 = Clamp(None=@n5) {desc=Prevent rounding errors on edges, maxDefault=1}
  n7 = Multiply(A=@n4, B=@n6)
  n8 = OneMinus(None=Multiply(A=@n4, B=OneMinus(None=@n6)))
  n9 = Multiply(A=@n6) {constB=2}
  FunctionOutput(None=LinearInterpolate(A=Min(A=@n7) {constB=0.5}, B=Max(A=@n8) {constB=0.5}, Alpha=Floor(None=@n9))) {description=The 9-slice UVs., outputName=UVs}
 ]
}
