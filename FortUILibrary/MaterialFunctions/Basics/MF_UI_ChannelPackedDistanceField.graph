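# UI material function, exposed to the library, that turns channel-packed distance field texture values into fill and outline masks.
# It takes up to four channels at once and outputs smoothstepped Fill and Outline, with Glow Min and Max controlling falloff, plus unsmoothed DF and DF Outline values.
# The outline folds the field around its 0.5 midpoint, offsets it by Outline Thickness, then saturates and remaps the result to 0–1.
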MaterialFunction /FortUILibrary/MaterialFunctions/Basics/MF_UI_ChannelPackedDistanceField
graph {
 properties {bExposeToLibrary=true, description=Control the distance field values of a texture}
 nodes [
  n1 = NamedRerouteUsage()
  FunctionOutput(None=@n1) {outputName=DF, sortPriority=2}
  n2 = NamedRerouteUsage()
  FunctionOutput(None=@n2) {outputName=DF Outline, sortPriority=3}
  FunctionOutput(None=SmoothStep(Min=NamedRerouteUsage(), Max=NamedRerouteUsage(), Value=@n1)) {outputName=Fill}
  FunctionOutput(None=SmoothStep(Min=NamedRerouteUsage(), Max=NamedRerouteUsage(), Value=@n2)) {outputName=Outline, sortPriority=1}
  n3 = FunctionInput(Preview=Constant() {r=0.5}) {
   bUsePreviewValueAsDefault=true
   description=Determines glow falloff strength. Can not go below GlowMin value or results will invert
   inputName=Glow Min
   inputType=FunctionInput_Scalar
   sortPriority=4
  }
  NamedRerouteDeclaration(None=@n3) {name=Glow Min, nodeColor=<rgba 0 0 0 1>}
  n4 = FunctionInput(Preview=Constant() {r=0.5}) {
   bUsePreviewValueAsDefault=true
   description=Determines glow falloff strength. Can not go below GlowMin value or results will invert
   inputName=Glow Max
   inputType=FunctionInput_Scalar
   sortPriority=5
  }
  NamedRerouteDeclaration(None=@n4) {name=Glow Max, nodeColor=<rgba 0 0 0 1>}
  n5 = NamedRerouteUsage()
  n6 = Constant() {desc=Input Low}
  n7 = Divide(A=Subtract(A=Saturate(None=@n5), B=@n6), B=Subtract(A=NamedRerouteUsage(), B=@n6))
  n8 = Constant() {desc=Target Low}
  n9 = Subtract(A=Constant() {desc=Target High, r=1}, B=@n8)
  NamedRerouteDeclaration(None=Add(A=Multiply(A=@n7, B=@n9), B=@n8)) {name=Remapped DF Channels Outline, nodeColor=<rgba 0 0 0 1>}
  n10 = FunctionInput(Preview=Constant()) {
   bUsePreviewValueAsDefault=true
   description=Pass one or multiple channels depending your needs
   inputName=RGBA Channels
   inputType=FunctionInput_Vector4
  }
  NamedRerouteDeclaration(None=@n10) {name=DF Channels, nodeColor=<rgba 0 0 0 1>}
  n11 = Abs(None=ConstantBiasScale(None=NamedRerouteUsage()) {bias=-0.5, scale=2})
  n12 = FunctionInput(Preview=Constant() {r=1}) {
   bUsePreviewValueAsDefault=true
   description=Determines the thickness of the stroke when using one of the outline outputs.
   inputName=Outline Thickness
   inputType=FunctionInput_Scalar
   sortPriority=6
  }
  n13 = NamedRerouteDeclaration(None=@n12) {name=Outline Thickness, nodeColor=<rgba 0 0 0 1>}
  NamedRerouteDeclaration(None=Subtract(A=OneMinus(None=@n11), B=OneMinus(None=@n13)) {constB=0.1}) {name=DF Channels Outline, nodeColor=<rgba 0 0 0 1>}
 ]
}
