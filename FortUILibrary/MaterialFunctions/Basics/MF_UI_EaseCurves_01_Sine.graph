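# UI material function that remaps a scalar 0-1 value through sinusoidal easing curves, exposed to the library.
# The input, defaulting to the horizontal normalized UI UV, feeds three saturated outputs: In, Out and In-Out.
# In and Out use quarter-wave cosine and sine of the halved value, with the cosine inverted.
# In-Out biases and scales a half-wave cosine into the 0-1 range.
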MaterialFunction /FortUILibrary/MaterialFunctions/Basics/MF_UI_EaseCurves_01_Sine
graph {
 properties {bExposeToLibrary=true, description=Outputs interpolated values using ease curves., previewMaterialDomain=MD_UI}
 nodes [
  Constant() {r=-0.5}
  Frac(None=Time())
  n1 = Cosine(None=NamedRerouteUsage()) {period=2}
  FunctionOutput(None=Saturate(None=ConstantBiasScale(None=@n1) {bias=-1, scale=-0.5})) {description=Ouputs an EaseInOut, outputName=In-Out, sortPriority=2}
  n2 = Multiply(A=NamedRerouteUsage()) {constB=0.5}
  n3 = Cosine(None=@n2) {period=2}
  FunctionOutput(None=Saturate(None=OneMinus(None=@n3))) {description=Ouputs an EaseIn, outputName=In}
  FunctionOutput(None=Saturate(None=Sine(None=@n2) {period=2})) {description=Ouputs an EaseOut, outputName=Out, sortPriority=1}
  n4 = FunctionInput(Preview=ComponentMask(None=MaterialFunctionCall() {materialFunction=&GetUserInterfaceUV}.Normalized UV) {r=true}) {bUsePreviewValueAsDefault=true, description=Inputs a value between 0-1., inputName="Value (0-1)", inputType=FunctionInput_Scalar}
  NamedRerouteDeclaration(None=@n4) {name=Value, nodeColor=<rgba 0 0 0 1>}
 ]
}
