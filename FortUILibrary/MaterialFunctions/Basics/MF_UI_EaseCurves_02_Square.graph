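# Material function in the UI library that outputs squared (quadratic) ease curves for a scalar 0-1 value, defaulting to horizontal normalized UI UV.
# In squares the value, and Out squares its inverse and inverts the result.
# In-Out mirrors a squared curve around the midpoint, using Round to pick the sign instead of branching, then remaps to 0-1.
# All three outputs are saturated.
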MaterialFunction /FortUILibrary/MaterialFunctions/Basics/MF_UI_EaseCurves_02_Square
graph {
 properties {bExposeToLibrary=true, description=Outputs interpolated values using ease curves., previewMaterialDomain=MD_UI}
 nodes [
  Constant() {r=-0.5}
  Frac(None=Time())
  n1 = NamedRerouteUsage()
  FunctionOutput(None=Saturate(None=Multiply(A=@n1, B=@n1))) {description=Ouputs an EaseIn, outputName=In, sortPriority=10}
  n2 = OneMinus(None=NamedRerouteUsage())
  n3 = Multiply(A=@n2, B=@n2)
  FunctionOutput(None=Saturate(None=OneMinus(None=@n3))) {description=Ouputs an EaseOut, outputName=Out, sortPriority=11}
  n4 = NamedRerouteUsage()
  n5 = OneMinus(None=Abs(None=ConstantBiasScale(None=@n4) {bias=-0.5, scale=2}))
  n6 = Multiply(A=LinearInterpolate(Alpha=Round(None=@n4)) {constA=-1}, B=OneMinus(None=Multiply(A=@n5, B=@n5)))
  FunctionOutput(None=Saturate(None=ConstantBiasScale(None=@n6))) {description=Ouputs an EaseInOut, outputName=In-Out, sortPriority=12}
  n7 = FunctionInput(Preview=ComponentMask(None=MaterialFunctionCall() {materialFunction=&GetUserInterfaceUV}.Normalized UV) {r=true}) {bUsePreviewValueAsDefault=true, description=Inputs a value between 0-1., inputName="Value (0-1)", inputType=FunctionInput_Scalar}
  NamedRerouteDeclaration(None=@n7) {name=Value, nodeColor=<rgba 0 0 0 1>}
 ]
}
