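# MF_UI_EvenOrOdd is a library-exposed UI material function that reports whether a scalar value is even or odd.
# It floors the input, multiplies it by 0.5, then takes the ceiling of the fractional part, returning 1 for odd and 0 for even.
# The branchless math avoids any If node, and the input uses a constant preview value as its default when unconnected.
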MaterialFunction /FortUILibrary/MaterialFunctions/Basics/MF_UI_EvenOrOdd
graph {
 properties {bExposeToLibrary=true, description=Outputs if a value is even or odd.}
 nodes [
  n1 = Constant() {r=4}
  n2 = Multiply(A=Floor(None=FunctionInput(Preview=@n1) {bUsePreviewValueAsDefault=true, description=Inputs an integer, inputName=Value, inputType=FunctionInput_Scalar})) {constB=0.5}
  FunctionOutput(None=Ceil(None=Frac(None=@n2))) {description="Returns 1 if odd, 0 if even"}
 ]
}
