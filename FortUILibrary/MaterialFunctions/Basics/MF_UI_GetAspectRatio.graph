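# UI material function that computes aspect ratio vectors from a widget's dimensions, for keeping proportions in UI materials.
# The Dimensions input defaults to Pixel Size from GetUserInterfaceUV and is split into X and Y.
# Horizontal and vertical ratios each pair a dimension quotient with 1, in opposite component order.
# Named reroutes carry the values to H Ratio, V Ratio, Min Ratio and Max Ratio outputs.
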MaterialFunction /FortUILibrary/MaterialFunctions/Basics/MF_UI_GetAspectRatio
graph {
 properties {
  bExposeToLibrary=true
  description="This MF is outputing the ratio of your widget base on the higher segment (width or height). It is using the Pixel Size output from the GetUserInterfaceUV MF "
 }
 nodes [
  FunctionOutput(None=NamedRerouteUsage()) {outputName=H Ratio}
  FunctionOutput(None=NamedRerouteUsage()) {outputName=V Ratio, sortPriority=1}
  FunctionOutput(None=Min(A=NamedRerouteUsage(), B=NamedRerouteUsage())) {outputName=Min Ratio, sortPriority=15}
  FunctionOutput(None=Max(A=NamedRerouteUsage(), B=NamedRerouteUsage())) {outputName=Max Ratio, sortPriority=10}
  n1 = NamedRerouteUsage()
  n2 = NamedRerouteUsage()
  n3 = Constant() {r=1}
  NamedRerouteDeclaration(None=AppendVector(A=Divide(A=@n1, B=@n2), B=@n3)) {name=hRatio, nodeColor=<rgba 0 0 0 1>}
  n4 = NamedRerouteUsage()
  n5 = NamedRerouteUsage()
  NamedRerouteDeclaration(None=AppendVector(A=@n3, B=Divide(A=@n4, B=@n5))) {name=vRatio, nodeColor=<rgba 0 0 0 1>}
  n6 = FunctionInput(Preview=MaterialFunctionCall() {materialFunction=&GetUserInterfaceUV}.Pixel Size) {
   bUsePreviewValueAsDefault=true
   description=Input to provide the pixel size or other dimensions such as view size
   inputName=Dimensions
   inputType=FunctionInput_Vector2
  }
  NamedRerouteDeclaration(None=ComponentMask(None=@n6) {r=true}) {name=Dimension X, nodeColor=<rgba 0 0 0 1>}
  NamedRerouteDeclaration(None=ComponentMask(None=@n6) {g=true}) {name=Dimension Y, nodeColor=<rgba 0 0 0 1>}
  Reroute() {desc="To keep Aspect Ratio, you need to use the \"Pixel Size\" from \"GetUserInterface UV\" MF"}
  Reroute() {desc="https://www.notion.so/fortnite-uxui/Keep-Aspect-Ratio-59226ab3ee41419d927f376ae5d053c2?pvs=4"}
 ]
}
