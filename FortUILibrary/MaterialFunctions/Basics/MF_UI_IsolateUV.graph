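# MF_UI_IsolateUV is a UI material function that outputs a mask isolating one cell of a UV grid.
# UV times Grid Size is rounded up to give cell numbers, which are compared with the rounded-up, one-based two-component Index.
# Saturated subtractions form branchless equality tests per axis, and the X and Y results multiply into a mask that is white only in the selected cell.
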MaterialFunction /FortUILibrary/MaterialFunctions/Basics/MF_UI_IsolateUV
graph {
 properties {bExposeToLibrary=true}
 nodes [
  n1 = TextureCoordinate()
  n2 = FunctionInput(Preview=Constant2Vector() {g=4, r=4}) {bUsePreviewValueAsDefault=true, description=Define Grid Size, inputName=Grid Size, inputType=FunctionInput_Vector2, sortPriority=1}
  n3 = Ceil(None=Multiply(A=FunctionInput(Preview=@n1) {bUsePreviewValueAsDefault=true, description=Input UV, inputName=UV, inputType=FunctionInput_Vector2}, B=@n2))
  n4 = FunctionInput(Preview=Constant2Vector() {g=1, r=1}) {bUsePreviewValueAsDefault=true, description=Define the selected index, inputName=Index, inputType=FunctionInput_Vector2, sortPriority=2}
  n5 = Ceil(None=@n4)
  n6 = Subtract(A=@n3, B=@n5)
  n7 = Constant() {r=1}
  n8 = Saturate(None=Subtract(A=@n3, B=Subtract(A=@n5, B=@n7)))
  n9 = OneMinus(None=Add(A=Saturate(None=@n6), B=OneMinus(None=@n8)))
  n10 = ComponentMask(None=@n9) {r=true}
  n11 = ComponentMask(None=@n9) {g=true}
  FunctionOutput(None=Saturate(None=Multiply(A=@n10, B=@n11))) {description=Output the isolated UV part}
 ]
}
