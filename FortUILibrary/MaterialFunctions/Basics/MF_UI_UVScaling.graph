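# UI material function that scales UVs to a UMG container while accounting for texture aspect ratio.
# It combines container size (pixel size by default) and texture size into an aspect-ratio pair, then derives Scale to Fit, Scale to Fill and single-axis ScaleX and ScaleY factors.
# UVs, normalized by default, are offset by a pivot defaulting to center, multiplied by each reciprocal factor, then shifted back, giving four outputs.
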MaterialFunction /FortUILibrary/MaterialFunctions/Basics/MF_UI_UVScaling
graph {
 properties {bExposeToLibrary=true, description=Scales the UVs to the Container size in UMG, libraryCategoriesText [UI Library]}
 nodes [
  n1 = NamedRerouteUsage()
  n2 = Constant() {r=1}
  n3 = AppendVector(A=ComponentMask(None=NamedRerouteUsage()) {r=true}, B=@n2) {desc=Scale to Fill Y}
  n4 = AppendVector(A=@n2, B=ComponentMask(None=NamedRerouteUsage()) {g=true}) {desc=Scale to Fill X}
  n5 = Divide(B=Max(A=Max(A=@n3, B=@n4) {desc=Scale To Fill}) {constB=0})
  FunctionOutput(None=Add(A=Multiply(A=@n1, B=@n5), B=NamedRerouteUsage())) {description="Scaled to Fill UVs ", outputName=Scaled To Fill, sortPriority=2}
  n6 = NamedRerouteUsage()
  n7 = NamedRerouteUsage()
  n8 = Divide(B=Max(A=Saturate(None=@n7) {desc=Scale To Fit}) {constB=0})
  FunctionOutput(None=Add(A=Multiply(A=@n6, B=@n8), B=NamedRerouteUsage())) {desc="Outputs the scaled to fit UVs ", description=Outputs the Scaled to Fit UVs, outputName=Scaled to Fit, sortPriority=1}
  n9 = NamedRerouteUsage()
  n10 = Divide(B=Max(A=@n3) {constB=0})
  FunctionOutput(None=Add(A=Multiply(A=@n9, B=@n10), B=NamedRerouteUsage())) {description=Outputs the UVs scaled to fit X, outputName=ScaleY, sortPriority=4}
  n11 = NamedRerouteUsage()
  n12 = Divide(B=Max(A=@n4) {constB=0})
  FunctionOutput(None=Add(A=Multiply(A=@n11, B=@n12), B=NamedRerouteUsage())) {description=Outputs the UVs scaled to fit Y, outputName=ScaleX, sortPriority=3}
  n13 = MaterialFunctionCall() {materialFunction=&GetUserInterfaceUV}
  n14 = FunctionInput(Preview=@n13.Normalized UV) {
   bUsePreviewValueAsDefault=true
   description=UV input by default uses Normalized UVs
   inputName=UV
   inputType=FunctionInput_Vector2
   sortPriority=1
  }
  n15 = FunctionInput(Preview=Constant2Vector() {g=0.5, r=0.5}) {
   bUsePreviewValueAsDefault=true
   description="Default is [.5, .5] to scale from the center"
   inputName=ScalePivotPoint
   inputType=FunctionInput_Vector2
   sortPriority=3
  }
  NamedRerouteDeclaration(None=Subtract(A=@n14, B=@n15)) {name=OffsetUVs, nodeColor=<rgba 0 0 0 1>}
  n16 = FunctionInput(Preview=@n13.Pixel Size) {
   bUsePreviewValueAsDefault=true
   description="The size of the container. By default, it uses the Pixel Size. But you can use that input to provide different container sizes such as the view size."
   inputName=ContainerSize
   inputType=FunctionInput_Vector2
   sortPriority=10
  }
  n17 = ComponentMask(None=@n16) {g=true}
  n18 = ComponentMask(None=@n16) {r=true}
  n19 = Divide(A=@n17, B=@n18)
  n20 = Divide(A=@n18, B=@n17)
  n21 = FunctionInput(Preview=Constant2Vector() {g=1, r=1}) {
   bUsePreviewValueAsDefault=true
   description=Input the texture size as a vector2
   inputName=TextureSize
   inputType=FunctionInput_Vector2
   sortPriority=2
  }
  n22 = ComponentMask(None=@n21) {r=true}
  n23 = ComponentMask(None=@n21) {g=true}
  n24 = Divide(A=@n22, B=@n23)
  n25 = Divide(A=@n23, B=@n22)
  NamedRerouteDeclaration(None=Multiply(A=AppendVector(A=@n19, B=@n20), B=AppendVector(A=@n24, B=@n25))) {name=AspectRatios, nodeColor=<rgba 0 0 0 1>}
  NamedRerouteDeclaration(None=@n15) {name=PivotPoints, nodeColor=<rgba 0 0 0 1>}
 ]
}
