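# MF_UI_Pattern is a library material function for UI that turns input UVs into repeating pattern UVs with separate X and Y tiling.
# It multiplies UVs by the tiling vector to output Raw UVs, then applies Frac for per-cell Pattern UVs.
# MF_UI_GetAspectRatio derives a Min Ratio from the tiling, which MF_UI_UVScaling uses to fit the cell UVs as Scaled Pattern UVs that retain pattern size.
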MaterialFunction /FortUILibrary/MaterialFunctions/Basics/MF_UI_Pattern
graph {
 properties {bExposeToLibrary=true, previewMaterialDomain=MD_UI}
 nodes [
  n1 = FunctionInput(Preview=TextureCoordinate()) {bUsePreviewValueAsDefault=true, description=Define the UVs, inputName=UV, inputType=FunctionInput_Vector2}
  n2 = FunctionInput() {
   bUsePreviewValueAsDefault=true
   description=Amount of subdivisions on the x-axis of your pattern
   inputName=Tiling X
   inputType=FunctionInput_Scalar
   sortPriority=1
   previewValue=<2 0 0 1>
  }
  n3 = FunctionInput() {
   bUsePreviewValueAsDefault=true
   description=Amount of subdivisions on the y-axis of your pattern
   inputName=Tiling Y
   inputType=FunctionInput_Scalar
   sortPriority=2
   previewValue=<2 0 0 1>
  }
  n4 = AppendVector(A=@n2, B=@n3)
  n5 = Frac(None=FunctionOutput(None=Multiply(A=@n1, B=@n4) {constB=10}) {description=These are the multiplied pattern value UVs, outputName=Raw UVs, sortPriority=3})
  FunctionOutput(None=@n5) {description=Fractional Pattern UVs, outputName=Pattern UVs}
  FunctionOutput(None=MaterialFunctionCall(
    UV=@n5
    TextureSize=MaterialFunctionCall!(Dimensions=@n4) {materialFunction=&/MF_UI_GetAspectRatio}.Min Ratio
   ) {materialFunction=&MF_UI_UVScaling}.Scaled to Fit) {description=Scaled to container UVs to retain pattern size, outputName=Scaled Pattern UVs, sortPriority=1}
 ]
}
