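# UI Library material function that warps UVs into a rotatable perspective view, defaulting to the 9-slice UV from GetUserInterfaceUV.
# It rotates the UVs by Perspective Rotation in full turns, then remaps the red channel to a centered gradient.
# Perspective Amount blends that gradient from 1 to form a divisor.
# Dividing the centered UVs by it and re-adding the 0.5 offset fakes perspective without any matrix projection.
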MaterialFunction /FortUILibrary/MaterialFunctions/Basics/MF_UI_Perspective
graph {
 properties {bExposeToLibrary=true, description=Rotate the UVs to generate a perspective view, libraryCategoriesText [UI Library]}
 nodes [
  n1 = FunctionInput(Preview=Constant()) {
   bUsePreviewValueAsDefault=true
   description=Define the rotation
   inputName=Perspective Rotation
   inputType=FunctionInput_Scalar
   sortPriority=2
  }
  n2 = ConstantBiasScale(None=Rotator(Coordinate=NamedRerouteUsage(), Time=@n1) {speed=6.2832}) {bias=-0.5, scale=2}
  n3 = FunctionInput(Preview=Constant() {r=0.4}) {
   bUsePreviewValueAsDefault=true
   description=Defines the intensity of the perspective effect
   inputName=Perspective Amount
   inputType=FunctionInput_Scalar
   sortPriority=2
  }
  n4 = Divide(A=Subtract(A=NamedRerouteUsage()) {constB=0.5}, B=LinearInterpolate(B=ComponentMask(None=@n2) {r=true}, Alpha=@n3) {constA=1})
  FunctionOutput(None=NamedRerouteDeclaration(None=Add(A=@n4) {constB=0.5}) {name=Perspective UVs, nodeColor=<rgba 0 0 0 1>}) {description=Return the UVs with a perspective rotation, outputName=Perspective UVs}
  NamedRerouteDeclaration(None=FunctionInput(Preview=MaterialFunctionCall() {materialFunction=&GetUserInterfaceUV}.9-Slice UV) {bUsePreviewValueAsDefault=true, description=Define the UV, inputName=UV, inputType=FunctionInput_Vector2}) {name=UV, nodeColor=<rgba 0 0 0 1>}
 ]
}
