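# UI library material function that converts a pixel value into a normalized value relative to the X or Y texture size.
# An X-or-Y input blends between width and height from the Pixel Size input, which defaults to GetUserInterfaceUV's output.
# The value is divided by that size and multiplied by the MF_UI_ViewportScale factor.
# A second output floors the result at one unscaled pixel, so thin values do not shrink below 1 px.
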MaterialFunction /FortUILibrary/MaterialFunctions/Basics/MF_UI_PixelsToNormalized
graph {
 properties {
  bExposeToLibrary=true
  description=Convert input Value in px into normalized value in relation to X or Y texture size.
  libraryCategoriesText [UI Library]
 }
 nodes [
  n1 = FunctionInput(Preview=Constant()) {
   bUsePreviewValueAsDefault=true
   description="0 (deafult) or 1.   0: relative to X texture size, 1: relative to Y texture size."
   inputName=X or Y
   inputType=FunctionInput_Scalar
   sortPriority=10
  }
  n2 = Max(A=LinearInterpolate(A=NamedRerouteUsage(), B=NamedRerouteUsage(), Alpha=@n1))
  n3 = Constant()
  n4 = Multiply(
   A=Divide(A=FunctionInput(Preview=@n3) {description=Value in pixels., inputName=Value PX, inputType=FunctionInput_Scalar}, B=@n2)
   B=NamedRerouteUsage()
  )
  FunctionOutput(None=Max(A=@n4, B=Divide(B=@n2))) {
   description=Outputs the ratio between Value and X or Y texture size and doesn't scale down normalized Value below 1 px.
   outputName="Normalized (>1px)"
   sortPriority=2
  }
  FunctionOutput(None=@n4) {description=Outputs the ratio between Value and X or Y texture size., outputName=Normalized, sortPriority=1}
  n5 = FunctionInput(Preview=MaterialFunctionCall() {materialFunction=&GetUserInterfaceUV}.Pixel Size) {bUsePreviewValueAsDefault=true, description=Pixel Size, inputName=UV Pixel Size, inputType=FunctionInput_Vector2, sortPriority=5}
  NamedRerouteDeclaration(None=ComponentMask(None=@n5) {r=true}) {name=TextureWidth_PX, nodeColor=<rgba 0 0 0 1>}
  NamedRerouteDeclaration(None=ComponentMask(None=@n5) {g=true}) {name=TextureHeight_PX, nodeColor=<rgba 0 0 0 1>}
  NamedRerouteDeclaration(None=MaterialFunctionCall() {materialFunction=&/MF_UI_ViewportScale}.Viewport Scale Factor) {name=ViewSizeCompensation, nodeColor=<rgba 0 0 0 1>}
 ]
}
