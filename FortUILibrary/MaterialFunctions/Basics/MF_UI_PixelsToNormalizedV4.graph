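# UI material library function that converts a Vector4 of pixel values into normalized values against a reference size, defaulting to the widget's pixel size.
# "Normalized" divides by the shorter reference side, guarded against zero, then multiplies by viewport DPI scale.
# "Per-Axis Normalized" also applies per-axis ratios that correct the longer axis (RB horizontal, GA vertical), keeping padding consistent across aspect ratios.
# A separate direct divide-by-width/height branch feeds neither output.
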MaterialFunction /FortUILibrary/MaterialFunctions/Basics/MF_UI_PixelsToNormalizedV4
graph {
 expressions {
  MaterialExpressionComponentMask_0 {
   class=MaterialExpressionComponentMask
   inputs [None {from=MaterialExpressionNamedRerouteUsage_0}]
   properties {b=true, desc=H in pixels, r=true}
  }
  MaterialExpressionComponentMask_1 {class=MaterialExpressionComponentMask, inputs [None {from=MaterialExpressionMax_1}], properties {r=true}}
  MaterialExpressionComponentMask_2 {class=MaterialExpressionComponentMask, inputs [None {from=MaterialExpressionMax_1}], properties {g=true}}
  MaterialExpressionComponentMask_3 {
   class=MaterialExpressionComponentMask
   inputs [None {from=MaterialExpressionNamedRerouteUsage_0}]
   properties {a=true, desc=V in pixels, g=true}
  }
  MaterialExpressionComponentMask_4 {class=MaterialExpressionComponentMask, inputs [None {from=MaterialExpressionMax_2}], properties {r=true}}
  MaterialExpressionComponentMask_5 {class=MaterialExpressionComponentMask, inputs [None {from=MaterialExpressionMax_2}], properties {g=true}}
  MaterialExpressionConstant4Vector_0 {class=MaterialExpressionConstant4Vector, inputs [], properties {constant=<rgba 10 10 10 10>}}
  MaterialExpressionConstant_0 {class=MaterialExpressionConstant, properties {r=10}}
  MaterialExpressionConvert_1 {
   class=MaterialExpressionConvert
   inputs [
    None {from=MaterialExpressionDivide_0}
    None {from=MaterialExpressionDivide_1}
   ]
   properties {
    desc=correction factor per axis
    nodeName=MakeFloat4
    convertInputs [
     {
      type=Vector2
      expressionInput {
       expression=&self:MaterialExpressionDivide_0
       inputName=None
       mask=0
       maskA=0
       maskB=0
       maskG=0
       maskR=0
       outputIndex=0
      }
     }
     {
      type=Vector2
      expressionInput {
       expression=&self:MaterialExpressionDivide_1
       inputName=None
       mask=0
       maskA=0
       maskB=0
       maskG=0
       maskR=0
       outputIndex=0
      }
     }
    ]
    convertOutputs [{type=Vector4, defaultValue=<rgba 0 0 0 1>}]
   }
  }
  MaterialExpressionConvert_2 {
   class=MaterialExpressionConvert
   inputs [
    None {from=MaterialExpressionNamedRerouteDeclaration_3}
    None {from=MaterialExpressionNamedRerouteDeclaration_4}
   ]
   properties {
    desc="correction factor per axis\r\nRatio X applied to RB (H)\r\nRatio Y applied to GA (V)"
    nodeName=MakeFloat4
    convertInputs [
     {
      type=Scalar
      expressionInput {
       expression=&self:MaterialExpressionNamedRerouteDeclaration_3
       inputName=None
       mask=0
       maskA=0
       maskB=0
       maskG=0
       maskR=0
       outputIndex=0
      }
     }
     {
      type=Scalar
      expressionInput {
       expression=&self:MaterialExpressionNamedRerouteDeclaration_4
       inputName=None
       mask=0
       maskA=0
       maskB=0
       maskG=0
       maskR=0
       outputIndex=0
      }
     }
    ]
    convertOutputs [{type=Vector4, defaultValue=<rgba 0 0 0 1>}]
   }
  }
  MaterialExpressionDivide_0 {
   class=MaterialExpressionDivide
   inputs [
    A {from=MaterialExpressionComponentMask_0}
    B {from=MaterialExpressionComponentMask_4}
   ]
  }
  MaterialExpressionDivide_1 {
   class=MaterialExpressionDivide
   inputs [
    A {from=MaterialExpressionComponentMask_3}
    B {from=MaterialExpressionComponentMask_5}
   ]
  }
  MaterialExpressionDivide_2 {
   class=MaterialExpressionDivide
   inputs [
    A {from=MaterialExpressionNamedRerouteUsage_1}
    B {from=MaterialExpressionNamedRerouteUsage_5}
   ]
  }
  MaterialExpressionDivide_3 {
   class=MaterialExpressionDivide
   inputs [
    A {from=MaterialExpressionNamedRerouteUsage_1}
    B {from=MaterialExpressionNamedRerouteUsage_6}
   ]
  }
  MaterialExpressionDivide_7 {
   class=MaterialExpressionDivide
   inputs [
    A {from=MaterialExpressionNamedRerouteDeclaration_0}
    B {from=MaterialExpressionNamedRerouteDeclaration_1}
   ]
   properties {constB=1, desc="all four channels ratios\r\nare “per‑min-axis”."}
  }
  MaterialExpressionFunctionInput_0 {
   class=MaterialExpressionFunctionInput
   inputs [Preview {from="", output=""}]
   properties {
    bUsePreviewValueAsDefault=true
    description=Any float or Vector4 value set in Pixels.
    inputName=Value in Pixels
    inputType=FunctionInput_Vector4
    previewValue=<0 0 0 0>
   }
  }
  MaterialExpressionFunctionInput_2 {
   class=MaterialExpressionFunctionInput
   inputs [Preview {from=MaterialExpressionMaterialFunctionCall_3, output=Pixel Size}]
   properties {
    bUsePreviewValueAsDefault=true
    description="Widget Size, Texture Size, Viewport Size..."
    inputName=Pixel Reference
    inputType=FunctionInput_Vector2
    sortPriority=1
   }
  }
  MaterialExpressionFunctionOutput_3 {
   class=MaterialExpressionFunctionOutput
   inputs [None {from=MaterialExpressionMultiply_2, output=""}]
   properties {
    description="Output Normalized vector4 values but with an additional correction factor applied to the current longest segment only, scaled by the widget’s shortest segment. \r\n\r\nNote: This helps normalize padding when aspect ratio differs."
    outputName="Per‑Axis Normalized (V4)"
    sortPriority=1
   }
  }
  MaterialExpressionFunctionOutput_35 {
   class=MaterialExpressionFunctionOutput
   inputs [None {from=MaterialExpressionMultiply_10, output=""}]
   properties {
    description="Output Normalized Vector4 values uniformly scaled by the widget’s shortest segment (min(W, H)).\r\n\r\nNote: Padding will vary with aspect ratio. Use per-Axis output instead."
    outputName="Normalized (V4)"
   }
  }
  MaterialExpressionMaterialFunctionCall_0 {
   class=MaterialExpressionMaterialFunctionCall
   inputs []
   properties {
    desc="The current DPI Scale being applied \r\nto the viewport and all the widgets. "
    materialFunction=&/MF_UI_ViewportScale
   }
  }
  MaterialExpressionMaterialFunctionCall_1 {class=MaterialExpressionMaterialFunctionCall, inputs [], properties {materialFunction=&GetUserInterfaceUV}}
  MaterialExpressionMaterialFunctionCall_2 {
   class=MaterialExpressionMaterialFunctionCall
   inputs []
   properties {
    desc="The current DPI Scale being applied \r\nto the viewport and all the widgets. "
    materialFunction=&MF_UI_ViewportScale
   }
  }
  MaterialExpressionMaterialFunctionCall_3 {class=MaterialExpressionMaterialFunctionCall, inputs [], properties {materialFunction=&GetUserInterfaceUV}}
  MaterialExpressionMaterialFunctionCall_5 {
   class=MaterialExpressionMaterialFunctionCall
   inputs []
   properties {
    desc="The current DPI Scale being applied \r\nto the viewport and all the widgets. "
    materialFunction=&MF_UI_ViewportScale
   }
  }
  MaterialExpressionMax_1 {
   class=MaterialExpressionMax
   inputs [
    A {from=MaterialExpressionFunctionInput_2}
    B {from=""}
   ]
   properties {desc="Minimum Pixel size is 1px\r\nAvoid dividing by 0"}
  }
  MaterialExpressionMax_2 {
   class=MaterialExpressionMax
   inputs [
    A {from=MaterialExpressionMaterialFunctionCall_1, output=Pixel Size}
    B {from=""}
   ]
   properties {desc="Minimum Pixel size is 1px\r\nAvoid dividing by 0"}
  }
  MaterialExpressionMin_0 {
   class=MaterialExpressionMin
   inputs [
    A {from=MaterialExpressionNamedRerouteDeclaration_17, output=""}
    B {from=MaterialExpressionNamedRerouteDeclaration_18, output=""}
   ]
   properties {desc=Get Safe Min-axis}
  }
  MaterialExpressionMultiply_0 {
   class=MaterialExpressionMultiply
   inputs [
    A {from=MaterialExpressionConvert_2}
    B {from=MaterialExpressionNamedRerouteUsage_4}
   ]
   properties {desc="apply the correction factor \r\nto the longest axis only.\r\nThis is why this is a per-axis\r\nnormalize."}
  }
  MaterialExpressionMultiply_1 {
   class=MaterialExpressionMultiply
   inputs [
    A {from=MaterialExpressionConvert_1}
    B {from=MaterialExpressionMaterialFunctionCall_0, output=Viewport Scale Factor}
   ]
   properties {desc="DPI‑Adjusted Pixels"}
  }
  MaterialExpressionMultiply_10 {
   class=MaterialExpressionMultiply
   inputs [
    A {from=MaterialExpressionNamedRerouteDeclaration_6}
    B {from=MaterialExpressionMaterialFunctionCall_5, output=Viewport Scale Factor}
   ]
   properties {desc="DPI‑Adjusted Pixels"}
  }
  MaterialExpressionMultiply_2 {
   class=MaterialExpressionMultiply
   inputs [
    A {from=MaterialExpressionMultiply_0}
    B {from=MaterialExpressionMaterialFunctionCall_2, output=Viewport Scale Factor}
   ]
   properties {desc="DPI‑Adjusted Pixels"}
  }
  MaterialExpressionNamedRerouteDeclaration_0 {
   class=MaterialExpressionNamedRerouteDeclaration
   inputs [None {from=MaterialExpressionFunctionInput_0}]
   properties {name=Value in Pixels, nodeColor=<rgba 0 0 0 1>}
  }
  MaterialExpressionNamedRerouteDeclaration_1 {
   class=MaterialExpressionNamedRerouteDeclaration
   inputs [None {from=MaterialExpressionMin_0}]
   properties {name=Safe Min-Axis, nodeColor=<rgba 0.0246 0.2691 0 1>}
  }
  MaterialExpressionNamedRerouteDeclaration_17 {
   class=MaterialExpressionNamedRerouteDeclaration
   inputs [None {from=MaterialExpressionComponentMask_1}]
   properties {name=safe pixelSize.x, nodeColor=<rgba 0 0.8125 1 1>}
  }
  MaterialExpressionNamedRerouteDeclaration_18 {
   class=MaterialExpressionNamedRerouteDeclaration
   inputs [None {from=MaterialExpressionComponentMask_2}]
   properties {name=safe pixelSize.y, nodeColor=<rgba 1 0.5391 0 1>}
  }
  MaterialExpressionNamedRerouteDeclaration_3 {
   class=MaterialExpressionNamedRerouteDeclaration
   inputs [None {from=MaterialExpressionDivide_2}]
   properties {name=Axis.x Ratio, nodeColor=<rgba 0 0.7422 1 1>}
  }
  MaterialExpressionNamedRerouteDeclaration_4 {
   class=MaterialExpressionNamedRerouteDeclaration
   inputs [None {from=MaterialExpressionDivide_3}]
   properties {name=Axis.y Ratio, nodeColor=<rgba 1 0.3297 0 1>}
  }
  MaterialExpressionNamedRerouteDeclaration_6 {
   class=MaterialExpressionNamedRerouteDeclaration
   inputs [None {from=MaterialExpressionDivide_7}]
   properties {
    name="Short‑Axis Normalized"
    desc="Normalizes each Vector4 value uniformly\r\n by the widget’s shortest segment (min(W, H))."
    nodeColor=<rgba 1 0 0.2813 1>
   }
  }
  MaterialExpressionNamedRerouteUsage_0 {class=MaterialExpressionNamedRerouteUsage}
  MaterialExpressionNamedRerouteUsage_1 {class=MaterialExpressionNamedRerouteUsage}
  MaterialExpressionNamedRerouteUsage_4 {class=MaterialExpressionNamedRerouteUsage}
  MaterialExpressionNamedRerouteUsage_5 {class=MaterialExpressionNamedRerouteUsage}
  MaterialExpressionNamedRerouteUsage_6 {class=MaterialExpressionNamedRerouteUsage}
 }
 properties {
  bExposeToLibrary=true
  description="Convert input Value(s) set in pixels into normalized value(s) in relation to the Width and the Height Widget/Texture size (in pixels)."
  previewMaterialDomain=MD_UI
  libraryCategoriesText [UI Library]
 }
}
