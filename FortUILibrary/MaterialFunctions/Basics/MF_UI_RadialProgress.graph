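# UI material function that generates a radial progress sector mask for timers and bars, with inputs for UVs, Progress, StartValue and Range.
# A Rotator turns the UVs by the start angle, then a cheap vector-to-radial function yields an angle, inverted when Range is negative for counter-clockwise fill.
# Scaled Progress is added to that angle, thresholded with anti-aliasing, and multiplied by a Progress-above-zero test to blank the empty state.
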MaterialFunction /FortUILibrary/MaterialFunctions/Basics/MF_UI_RadialProgress
graph {
 expressions {
  MaterialExpressionAdd_0 {
   class=MaterialExpressionAdd
   inputs [
    A {from=MaterialExpressionReroute_3}
    B {from=MaterialExpressionLinearInterpolate_2}
   ]
  }
  MaterialExpressionConstant_0 {class=MaterialExpressionConstant}
  MaterialExpressionConstant_1 {class=MaterialExpressionConstant}
  MaterialExpressionConstant_3 {class=MaterialExpressionConstant, properties {r=1}}
  MaterialExpressionFunctionInput_0 {
   class=MaterialExpressionFunctionInput
   inputs [Preview {from=MaterialExpressionMaterialFunctionCall_0, output=Normalized UV}]
   properties {bUsePreviewValueAsDefault=true, description=Optional custom UV mapping., inputName=UVs, inputType=FunctionInput_Vector2}
  }
  MaterialExpressionFunctionInput_6 {
   class=MaterialExpressionFunctionInput
   inputs [Preview {from="", output=""}]
   properties {
    bUsePreviewValueAsDefault=true
    description="The starting direction in degrees of the meter (0-360) where 0 is up, 90 is right, 180 is down, and 270 is left. Default is 0."
    inputName=StartValue
    inputType=FunctionInput_Scalar
    sortPriority=3
   }
  }
  MaterialExpressionFunctionInput_7 {
   class=MaterialExpressionFunctionInput
   inputs [Preview {from="", output=""}]
   properties {
    bUsePreviewValueAsDefault=true
    description="The range (in degrees) of the progress bar. Positive values progress clockwise, negative values go counter-clockwise. Default is 360."
    inputName=Range
    inputType=FunctionInput_Scalar
    sortPriority=4
    previewValue=<360 0 0 1>
   }
  }
  MaterialExpressionFunctionInput_8 {
   class=MaterialExpressionFunctionInput
   inputs [Preview {from=MaterialExpressionMaterialFunctionCall_1, output=Result}]
   properties {
    bUsePreviewValueAsDefault=true
    description="The current progress (0-1). Default is DebugTimeSine."
    inputName=Progress
    inputType=FunctionInput_Scalar
    sortPriority=1
   }
  }
  MaterialExpressionFunctionOutput_0 {class=MaterialExpressionFunctionOutput, inputs [None {from=MaterialExpressionMultiply_3, output=""}]}
  MaterialExpressionLinearInterpolate_2 {
   class=MaterialExpressionLinearInterpolate
   inputs [
    A {from=MaterialExpressionMaterialFunctionCall_2, output=Vector Converted to Angle}
    B {from=MaterialExpressionOneMinus_0}
    Alpha {from=MaterialExpressionMaterialFunctionCall_5, output=IsLess}
   ]
  }
  MaterialExpressionMaterialFunctionCall_0 {class=MaterialExpressionMaterialFunctionCall, inputs [], properties {materialFunction=&GetUserInterfaceUV}}
  MaterialExpressionMaterialFunctionCall_1 {
   class=MaterialExpressionMaterialFunctionCall
   inputs [Speed {from=""}]
   properties {materialFunction=&DebugTimeSine}
  }
  MaterialExpressionMaterialFunctionCall_2 {
   class=MaterialExpressionMaterialFunctionCall
   inputs [Vector or UVs {from=MaterialExpressionRotator_1}]
   properties {materialFunction=&/MF_UI_CheaperVectorToRadial}
  }
  MaterialExpressionMaterialFunctionCall_3 {
   class=MaterialExpressionMaterialFunctionCall
   inputs [
    DistanceField {from=MaterialExpressionAdd_0}
    Threshold {from=MaterialExpressionConstant_3}
    Smoothing {from=""}
   ]
   properties {materialFunction=&MF_UI_AntiAliasing}
  }
  MaterialExpressionMaterialFunctionCall_4 {
   class=MaterialExpressionMaterialFunctionCall
   inputs [
    A {from=MaterialExpressionReroute_6}
    B {from=MaterialExpressionConstant_0}
   ]
   properties {materialFunction=&/MF_UI_BooleanOperations}
  }
  MaterialExpressionMaterialFunctionCall_5 {
   class=MaterialExpressionMaterialFunctionCall
   inputs [
    A {from=MaterialExpressionNamedRerouteUsage_0}
    B {from=MaterialExpressionConstant_1}
   ]
   properties {materialFunction=&MF_UI_BooleanOperations}
  }
  MaterialExpressionMultiply_3 {
   class=MaterialExpressionMultiply
   inputs [
    A {from=MaterialExpressionMaterialFunctionCall_4, output=IsGreater}
    B {from=MaterialExpressionReroute_2}
   ]
  }
  MaterialExpressionMultiply_4 {
   class=MaterialExpressionMultiply
   inputs [
    A {from=MaterialExpressionReroute_5}
    B {from=MaterialExpressionMultiply_0}
   ]
  }
  MaterialExpressionNamedRerouteDeclaration_0 {
   class=MaterialExpressionNamedRerouteDeclaration
   inputs [None {from=MaterialExpressionFunctionInput_7}]
   properties {name=Range, nodeColor=<rgba 0 0 0 1>}
  }
  MaterialExpressionNamedRerouteUsage_0 {class=MaterialExpressionNamedRerouteUsage}
  MaterialExpressionOneMinus_0 {class=MaterialExpressionOneMinus, inputs [None {from=MaterialExpressionMaterialFunctionCall_2, output=Vector Converted to Angle}]}
  MaterialExpressionReroute_0 {class=MaterialExpressionReroute, inputs [None {from=MaterialExpressionMaterialFunctionCall_3, output=Outer}]}
  MaterialExpressionReroute_2 {class=MaterialExpressionReroute, inputs [None {from=MaterialExpressionReroute_0}]}
  MaterialExpressionReroute_3 {class=MaterialExpressionReroute, inputs [None {from=MaterialExpressionMultiply_4}]}
  MaterialExpressionReroute_5 {class=MaterialExpressionReroute, inputs [None {from=MaterialExpressionFunctionInput_8}]}
  MaterialExpressionReroute_6 {class=MaterialExpressionReroute, inputs [None {from=MaterialExpressionReroute_5}]}
  MaterialExpressionRotator_1 {
   class=MaterialExpressionRotator
   inputs [
    Coordinate {from=MaterialExpressionFunctionInput_0, output=""}
    Time {from=MaterialExpressionFunctionInput_6, output=""}
   ]
   properties {speed=-0.0175}
  }
 }
 properties {
  description=Generates a radial progress sector. Useful for masking textures to make radial progress timers or bars.
  libraryCategoriesText [UILibrary]
 }
}
