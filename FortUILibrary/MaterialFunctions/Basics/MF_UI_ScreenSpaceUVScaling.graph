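# UI Library material function that scales UVs to the screen for a texture of a given size.
# It wraps MF_UI_UVScaling, feeding it screen UVs (viewport UV by default), texture size, a centered-by-default pivot, and ViewSize as the container.
# It outputs Scaled to Fit, Scaled to Fill, ScaleX and ScaleY UVs.
# The ScaleX output's description says fit Y and ScaleY's says fit X, opposite to their names.
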MaterialFunction /FortUILibrary/MaterialFunctions/Basics/MF_UI_ScreenSpaceUVScaling
graph {
 properties {bExposeToLibrary=true, description=Scales the UVs to the screen, libraryCategoriesText [UI Library]}
 nodes [
  n1 = FunctionInput(Preview=ScreenPosition().ViewportUV) {
   bUsePreviewValueAsDefault=true
   description="Set screen UV (Default is ScreenPosition/Viewport UV)"
   inputName=ScreenUV
   inputType=FunctionInput_Vector2
  }
  n2 = FunctionInput() {
   bUsePreviewValueAsDefault=true
   description=Input the texture size
   inputName=TextureSize
   inputType=FunctionInput_Vector2
   sortPriority=1
   previewValue=<1 1 0 1>
  }
  n3 = FunctionInput() {
   bUsePreviewValueAsDefault=true
   description=Scale from Center by default. Use to define where the pivot point influence the scale.
   inputName=ScalePivotPoint
   inputType=FunctionInput_Vector2
   sortPriority=2
   previewValue=<0.5 0.5 0 1>
  }
  n4 = MaterialFunctionCall(UV=@n1, TextureSize=@n2, ScalePivotPoint=@n3, ContainerSize=ViewSize()) {materialFunction=&MF_UI_UVScaling}
  FunctionOutput(None=@n4.Scaled to Fit) {desc="Outputs the scaled to fit UVs ", description=Outputs the Scaled to Fit UVs, outputName=Scaled to Fit, sortPriority=1}
  FunctionOutput(None=@n4.ScaleY) {description=Outputs the UVs scaled to fit X, outputName=ScaleY, sortPriority=4}
  FunctionOutput(None=@n4.ScaleX) {description=Outputs the UVs scaled to fit Y, outputName=ScaleX, sortPriority=3}
  FunctionOutput(None=@n4.Scaled To Fill) {description="Scaled to Fill UVs ", outputName=Scaled to Fill, sortPriority=2}
 ]
}
