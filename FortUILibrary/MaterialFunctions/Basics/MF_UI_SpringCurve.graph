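# UI material function that turns a 0-1 value (UI U coordinate by default) into spring-style interpolation, tuned by Stiffness and Damping.
# Exponential decay multiplies cosine and sine oscillations, forming damped curves through named reroutes.
# Lerps toward constant 1 and 0 produce the Spring Curve and the rebounding Rebounce Spring Curve outputs.
# Two debug outputs plot these curves with PlotFunctionOnGraph.
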MaterialFunction /FortUILibrary/MaterialFunctions/Basics/MF_UI_SpringCurve
graph {
 properties {bExposeToLibrary=true, description=Interpolates a value based on spring properties, previewMaterialDomain=MD_UI}
 nodes [
  n1 = LinearInterpolate(A=NamedRerouteUsage(), B=Constant(), Alpha=NamedRerouteUsage())
  FunctionOutput(None=MaterialFunctionCall!(
    X range Y range=Constant4Vector() {constant=<rgba 0 1 -1 1>}.RGBA
    Gradient Height=Constant()
    ShowGrid?
    Thickness
    Color
    "f(x)"=FunctionOutput(None=@n1) {description=Returns the interpolated value that rebounces back, outputName=Rebounce Spring Curve, sortPriority=5}
   ) {materialFunction=/Engine/Functions/Engine_MaterialFunctions01/Debug/PlotFunctionOnGraph}."f(x) = y") {description=Returns a debug preview of the rebound curve, outputName="Rebounce Curve Preview (Debug)", sortPriority=20}
  n2 = LinearInterpolate(A=NamedRerouteUsage(), B=Constant() {r=1}, Alpha=NamedRerouteUsage())
  FunctionOutput(None=MaterialFunctionCall(
    X range Y range=Constant4Vector() {constant=<rgba 0 1 0 2>}.RGBA
    Gradient Height=Constant()
    "f(x)"=FunctionOutput(None=@n2) {description=Returns the interpolated value, outputName=Spring Curve}
   ) {materialFunction=&PlotFunctionOnGraph}."f(x) = y") {description=Returns a debug preview of the curve, outputName="Curve Preview (Debug)", sortPriority=15}
  n3 = FunctionInput(Preview=ComponentMask(None=MaterialFunctionCall() {materialFunction=&GetUserInterfaceUV}.Normalized UV) {r=true}) {bUsePreviewValueAsDefault=true, description=Set the value between 0-1, inputName="Value (0-1)", inputType=FunctionInput_Scalar}
  NamedRerouteDeclaration(None=@n3) {name=Time, nodeColor=<rgba 0 0 0 1>}
  n4 = FunctionInput() {
   bUsePreviewValueAsDefault=true
   description=Defines how much resistance
   inputName=Damping
   inputType=FunctionInput_Scalar
   sortPriority=2
   previewValue=<4 0 0 1>
  }
  NamedRerouteDeclaration(None=@n4) {name=Damping, nodeColor=<rgba 0 0 0 1>}
  n5 = FunctionInput() {
   bUsePreviewValueAsDefault=true
   description=Defines how stiff the curve will be
   inputName=Stiffness
   inputType=FunctionInput_Scalar
   sortPriority=1
   previewValue=<18 0 0 1>
  }
  NamedRerouteDeclaration(None=@n5) {name=Stiffness, nodeColor=<rgba 0 0 0 1>}
  n6 = NamedRerouteUsage()
  n7 = Power(Base=NamedRerouteUsage(), Exp=Multiply(A=NamedRerouteUsage(), B=Multiply(A=@n6) {constB=-1}) {constB=-1})
  n8 = Multiply(A=NamedRerouteUsage(), B=NamedRerouteUsage())
  n9 = Cosine(None=@n8) {period=6.2832}
  NamedRerouteDeclaration(None=OneMinus(None=Multiply(A=@n7, B=@n9))) {name="Spring Curve (Cos)", nodeColor=<rgba 0 0 0 1>}
  NamedRerouteDeclaration(None=Multiply(A=@n7, B=Sine(None=@n8) {period=6.2832})) {name="Spring Curve (Sin)", nodeColor=<rgba 0 0 0 1>}
  NamedRerouteDeclaration(None=Constant() {r=2.7183}) {name=Euler Number, nodeColor=<rgba 0 0 0 1>}
 ]
}
