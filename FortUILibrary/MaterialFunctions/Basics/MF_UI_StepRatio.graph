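# UI material function that quantizes a 0-1 value into a set number of steps, four by default, with normalized UI U coordinate as the default value.
# It multiplies the value by the step count, which is floored at a tiny positive number, and truncates the result.
# Dividing by the count yields Stepped Ratio, while dividing by count minus one yields Stepped Gradient, both saturated.
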MaterialFunction /FortUILibrary/MaterialFunctions/Basics/MF_UI_StepRatio
graph {
 properties {bExposeToLibrary=true, description=Output a ratio that is divided by the number of steps}
 nodes [
  n1 = ComponentMask(None=MaterialFunctionCall() {materialFunction=&GetUserInterfaceUV}.Normalized UV) {r=true}
  n2 = FunctionInput(Preview=Constant() {r=4}) {bUsePreviewValueAsDefault=true, description=Define the numbers of intervals, inputName=Steps, inputType=FunctionInput_Scalar, sortPriority=1}
  n3 = Max(A=@n2) {constB=0.0001}
  n4 = Truncate!(None=Multiply(
    A=FunctionInput(Preview=@n1) {bUsePreviewValueAsDefault=true, description=Set a value, inputName="Value (0-1)", inputType=FunctionInput_Vector2}
    B=@n3
   )) {is=class MaterialExpressionTruncate}
  FunctionOutput(None=Saturate(None=Divide(A=@n4, B=@n3))) {description=Returns a segmented ratio, outputName=Stepped Ratio}
  n5 = Divide(A=Constant() {r=1}, B=Subtract(A=@n3))
  FunctionOutput(None=Saturate(None=Multiply(A=@n4, B=@n5))) {description=Returns a segmented gradient, outputName=Stepped Gradient, sortPriority=1}
 ]
}
