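# Deprecated UI material function that splits UVs into two side-by-side cells and mirrors the right cell against the left.
# Doubles U, takes the fraction, adds Offset, then lerps to a flipped copy using a floored, saturated U mask.
# MirrorY static switch selects the flipped axis: true flips U, false flips V.
# CustomRotator and ScaleUVsByCenter then apply Rotation and Scale; left and right masks are also output.
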MaterialFunction /FortUILibrary/MaterialFunctions/Basics/MF_UI_Symmetry_2x1_DEPRECATED
graph {
 properties {description=Mirrors visually the input horizontally or vertically.}
 nodes [
  n1 = Multiply(
   A=FunctionInput(Preview=TextureCoordinate()) {bUsePreviewValueAsDefault=true, description=Inputs UV, inputName=UV, inputType=FunctionInput_Vector2, sortPriority=1}
   B=Constant2Vector() {g=1, r=2}
  )
  n2 = FunctionInput(Preview=Constant2Vector()) {bUsePreviewValueAsDefault=true, description=Defines an offset, inputName=Offset, inputType=FunctionInput_Vector2, sortPriority=3}
  n3 = Add(A=Frac(None=@n1), B=@n2)
  n4 = MaterialFunctionCall(Float2=Add(A=OneMinus(None=@n3), B=Multiply(A=@n2) {constB=2})) {materialFunction=&BreakOutFloat2Components}
  n5 = MaterialFunctionCall(Float2=@n3) {materialFunction=&BreakOutFloat2Components}
  n6 = FunctionInput(Preview=StaticBool() {value=true}) {
   bUsePreviewValueAsDefault=true
   description=Switch between mirroring X or Y
   inputName=MirrorY
   inputType=FunctionInput_StaticBool
   sortPriority=32
  }
  n7 = Saturate(None=Floor(None=ComponentMask(None=@n1) {r=true}))
  n8 = LinearInterpolate(A=@n3, B=StaticSwitch(True=AppendVector(A=@n4.R, B=@n5.G), False=AppendVector(A=@n5.R, B=@n4.G), Value=@n6), Alpha=@n7)
  n9 = FunctionInput(Preview=Constant()) {bUsePreviewValueAsDefault=true, description=Define a rotation, inputName=Rotation, inputType=FunctionInput_Scalar, sortPriority=4}
  n10 = FunctionInput(Preview=Constant() {r=0.5}) {bUsePreviewValueAsDefault=true, description=Define a scale factor, inputName=Scale, inputType=FunctionInput_Vector2, sortPriority=5}
  FunctionOutput(None=MaterialFunctionCall(
    UVs=MaterialFunctionCall(UVs=@n8, "Rotation Angle (0-1)"=@n9) {materialFunction=&CustomRotator}.Rotated Values
    Texture Scale=@n10
   ) {materialFunction=&ScaleUVsByCenter}.UVs) {description=Ouputs a mirror of the input}
  FunctionOutput(None=@n7) {description=Outputs the right side mask of the symmetry, outputName=Right Side Mask, sortPriority=2}
  FunctionOutput(None=OneMinus(None=@n7)) {description=Outputs the left side mask of the symmetry, outputName=Left Side Mask, sortPriority=1}
 ]
}
