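# UI material function that mirrors UVs into a two-by-one symmetry, defaulting to user interface UVs and outputting Symmetry and Flipped Symmetry.
# One axis is remapped around its center and folded with Abs, with OneMinus giving the non-flipped variant.
# The other axis blends between original and inverted V through lerps, supported by a rounded symmetry mask.
# Scalar Mirror Y and Swizzle UV inputs select mirroring and component order without static switches.
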MaterialFunction /FortUILibrary/MaterialFunctions/Basics/MF_UI_Symmetry_2x1_New
graph {
 expressions {
  MaterialExpressionAbs_6 {class=MaterialExpressionAbs, inputs [None {from=MaterialExpressionConstantBiasScale_5, output=""}], properties {desc=Mirror UV}}
  MaterialExpressionComponentMask_1 {class=MaterialExpressionComponentMask, inputs [None {from=MaterialExpressionFunctionInput_0}], properties {r=true}}
  MaterialExpressionComponentMask_2 {class=MaterialExpressionComponentMask, inputs [None {from=MaterialExpressionFunctionInput_0}], properties {g=true}}
  MaterialExpressionConstantBiasScale_5 {class=MaterialExpressionConstantBiasScale, inputs [None {from=MaterialExpressionNamedRerouteUsage_22}], properties {bias=-0.5, scale=2}}
  MaterialExpressionConvert_7 {
   class=MaterialExpressionConvert
   inputs [
    None {from=MaterialExpressionAbs_6}
    None {from=MaterialExpressionNamedRerouteUsage_35}
   ]
   properties {
    nodeName=Break Float2
    convertInputs [
     {
      type=Scalar
      expressionInput {
       expression=&self:MaterialExpressionAbs_6
       inputName=None
       mask=0
       maskA=0
       maskB=0
       maskG=0
       maskR=0
       outputIndex=0
      }
     }
     {
      type=Scalar
      expressionInput {
       expression=&self:MaterialExpressionNamedRerouteUsage_35
       inputName=None
       mask=0
       maskA=0
       maskB=0
       maskG=0
       maskR=0
       outputIndex=0
      }
     }
    ]
    convertOutputs [{type=Vector2}, {type=Vector2}]
   }
  }
  MaterialExpressionConvert_8 {
   class=MaterialExpressionConvert
   inputs [
    None {from=MaterialExpressionOneMinus_19}
    None {from=MaterialExpressionNamedRerouteUsage_35}
   ]
   properties {
    nodeName=Break Float2
    convertInputs [
     {
      type=Scalar
      expressionInput {
       expression=&self:MaterialExpressionOneMinus_19
       inputName=None
       mask=0
       maskA=0
       maskB=0
       maskG=0
       maskR=0
       outputIndex=0
      }
     }
     {
      type=Scalar
      expressionInput {
       expression=&self:MaterialExpressionNamedRerouteUsage_35
       inputName=None
       mask=0
       maskA=0
       maskB=0
       maskG=0
       maskR=0
       outputIndex=0
      }
     }
    ]
    convertOutputs [{type=Vector2}, {type=Vector2}]
   }
  }
  MaterialExpressionFunctionInput_0 {
   class=MaterialExpressionFunctionInput
   inputs [Preview {from=MaterialExpressionMaterialFunctionCall_6, output=Normalized UV}]
   properties {bUsePreviewValueAsDefault=true, description=Inputs UV, inputName=UV, inputType=FunctionInput_Vector2}
  }
  MaterialExpressionFunctionInput_1 {
   class=MaterialExpressionFunctionInput
   inputs [Preview {from="", output=""}]
   properties {
    bUsePreviewValueAsDefault=true
    description=Switch between mirroring X or Y
    inputName=Mirror Y
    inputType=FunctionInput_Scalar
    sortPriority=5
    previewValue=<1 0 0 1>
   }
  }
  MaterialExpressionFunctionInput_2 {
   class=MaterialExpressionFunctionInput
   inputs [Preview {from="", output=""}]
   properties {bUsePreviewValueAsDefault=true, description=Flip the UVs, inputName=Swizzle UV, inputType=FunctionInput_Scalar, sortPriority=10}
  }
  MaterialExpressionFunctionOutput_3 {
   class=MaterialExpressionFunctionOutput
   inputs [None {from=MaterialExpressionLinearInterpolate_13, output=""}]
   properties {description=Output the UVs as a symmetry, outputName=Symmetry}
  }
  MaterialExpressionFunctionOutput_6 {
   class=MaterialExpressionFunctionOutput
   inputs [None {from=MaterialExpressionLinearInterpolate_12, output=""}]
   properties {description=Output the UVs as a flipped symmetry, outputName=Flipped Symmetry, sortPriority=5}
  }
  MaterialExpressionLinearInterpolate_12 {
   class=MaterialExpressionLinearInterpolate
   inputs [
    A {from=MaterialExpressionConvert_7}
    B {from=MaterialExpressionConvert_7}
    Alpha {from=MaterialExpressionNamedRerouteUsage_19}
   ]
  }
  MaterialExpressionLinearInterpolate_13 {
   class=MaterialExpressionLinearInterpolate
   inputs [
    A {from=MaterialExpressionConvert_8}
    B {from=MaterialExpressionConvert_8}
    Alpha {from=MaterialExpressionNamedRerouteUsage_20}
   ]
  }
  MaterialExpressionLinearInterpolate_17 {
   class=MaterialExpressionLinearInterpolate
   inputs [
    A {from=MaterialExpressionComponentMask_2}
    B {from=MaterialExpressionOneMinus_15}
    Alpha {from=MaterialExpressionNamedRerouteUsage_28}
   ]
  }
  MaterialExpressionLinearInterpolate_18 {
   class=MaterialExpressionLinearInterpolate
   inputs [
    A {from=MaterialExpressionLinearInterpolate_17}
    B {from=MaterialExpressionComponentMask_2}
    Alpha {from=MaterialExpressionNamedRerouteUsage_29}
   ]
  }
  MaterialExpressionMaterialFunctionCall_6 {class=MaterialExpressionMaterialFunctionCall, inputs [], properties {materialFunction=&GetUserInterfaceUV}}
  MaterialExpressionNamedRerouteDeclaration_0 {
   class=MaterialExpressionNamedRerouteDeclaration
   inputs [None {from=MaterialExpressionRound_0}]
   properties {name=Symmetry Mask, nodeColor=<rgba 0 0 0 1>}
  }
  MaterialExpressionNamedRerouteDeclaration_5 {
   class=MaterialExpressionNamedRerouteDeclaration
   inputs [None {from=MaterialExpressionComponentMask_1}]
   properties {name=X Axis, nodeColor=<rgba 0 0 0 1>}
  }
  MaterialExpressionNamedRerouteDeclaration_6 {
   class=MaterialExpressionNamedRerouteDeclaration
   inputs [None {from=MaterialExpressionLinearInterpolate_18}]
   properties {name=Y Axis, nodeColor=<rgba 0 0 0 1>}
  }
  MaterialExpressionNamedRerouteDeclaration_7 {
   class=MaterialExpressionNamedRerouteDeclaration
   inputs [None {from=MaterialExpressionFunctionInput_1}]
   properties {name=Mirror Y, nodeColor=<rgba 0 0 0 1>}
  }
  MaterialExpressionNamedRerouteDeclaration_8 {
   class=MaterialExpressionNamedRerouteDeclaration
   inputs [None {from=MaterialExpressionFunctionInput_2}]
   properties {name=Swizzle UV, nodeColor=<rgba 0 0 0 1>}
  }
  MaterialExpressionNamedRerouteUsage_15 {class=MaterialExpressionNamedRerouteUsage}
  MaterialExpressionNamedRerouteUsage_19 {class=MaterialExpressionNamedRerouteUsage}
  MaterialExpressionNamedRerouteUsage_20 {class=MaterialExpressionNamedRerouteUsage}
  MaterialExpressionNamedRerouteUsage_22 {class=MaterialExpressionNamedRerouteUsage}
  MaterialExpressionNamedRerouteUsage_28 {class=MaterialExpressionNamedRerouteUsage}
  MaterialExpressionNamedRerouteUsage_29 {class=MaterialExpressionNamedRerouteUsage}
  MaterialExpressionNamedRerouteUsage_35 {class=MaterialExpressionNamedRerouteUsage}
  MaterialExpressionOneMinus_15 {class=MaterialExpressionOneMinus, inputs [None {from=MaterialExpressionComponentMask_2}]}
  MaterialExpressionOneMinus_19 {class=MaterialExpressionOneMinus, inputs [None {from=MaterialExpressionAbs_6}]}
  MaterialExpressionRound_0 {class=MaterialExpressionRound, inputs [None {from=MaterialExpressionNamedRerouteUsage_15, output=""}], properties {}}
 }
 properties {
  bExposeToLibrary=true
  description=Mirrors the UVs horizontally or vertically
  previewMaterialDomain=MD_UI
  libraryCategoriesText [UI_Library]
 }
}
