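# UI library material function that mirrors UVs horizontally and vertically to produce 2x2 symmetry.
# The input UV, defaulting to normalized UI UV, is remapped around its center and passed through Abs for the Flipped Symmetry output.
# A OneMinus of that result feeds the Symmetry output.
# Each output passes through a Break Float2 conversion and a lerp driven by the scalar Swizzle UV input, which flips the UVs.
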MaterialFunction /FortUILibrary/MaterialFunctions/Basics/MF_UI_Symmetry_2x2_New
graph {
 properties {
  bExposeToLibrary=true
  description=Mirrors the UVs horizontally and vertically
  previewMaterialDomain=MD_UI
  libraryCategoriesText [UI_Library]
 }
 nodes [
  MaterialExpressionAbs_1 = Abs(None=ConstantBiasScale(None=NamedRerouteUsage()) {bias=-0.5, scale=2}) {desc=Mirror UV}
  MaterialExpressionOneMinus_2 = OneMinus(None=@MaterialExpressionAbs_1)
  n1 = Convert(None=@MaterialExpressionOneMinus_2) {
   nodeName=Break Float2
   convertInputs [{
     type=Vector2
     expressionInput {
      expression=&self:MaterialExpressionOneMinus_2
      inputName=None
      mask=0
      maskA=0
      maskB=0
      maskG=0
      maskR=0
      outputIndex=0
     }
    }]
   convertOutputs [{type=Vector2}, {type=Vector2}]
  }
  FunctionOutput(None=LinearInterpolate(A=@n1, B=@n1, Alpha=NamedRerouteUsage())) {description=Output the UVs as a symmetry, outputName=Symmetry}
  n2 = Convert(None=@MaterialExpressionAbs_1) {
   nodeName=Break Float2
   convertInputs [{
     type=Vector2
     expressionInput {
      expression=&self:MaterialExpressionAbs_1
      inputName=None
      mask=0
      maskA=0
      maskB=0
      maskG=0
      maskR=0
      outputIndex=0
     }
    }]
   convertOutputs [{type=Vector2}, {type=Vector2}]
  }
  FunctionOutput(None=LinearInterpolate(A=@n2, B=@n2, Alpha=NamedRerouteUsage())) {description=Output the UVs as a flipped symmetry, outputName=Flipped Symmetry, sortPriority=5}
  NamedRerouteDeclaration(None=FunctionInput(Preview=MaterialFunctionCall() {materialFunction=&GetUserInterfaceUV}.Normalized UV) {bUsePreviewValueAsDefault=true, description=Input the UV, inputName=UV, inputType=FunctionInput_Vector2}) {name=UV, nodeColor=<rgba 0 0 0 1>}
  n3 = FunctionInput() {bUsePreviewValueAsDefault=true, description=Flip the UVs, inputName=Swizzle UV, inputType=FunctionInput_Scalar, sortPriority=10}
  NamedRerouteDeclaration(None=@n3) {name=Swizzle UV, nodeColor=<rgba 0 0 0 1>}
 ]
}
