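# UI material function that generates UVs repeated rotationally around a center point for a given number of sides, defaulting to normalized UI UVs.
# A cheaper vector-to-radial function turns inverted UVs into an angle, which is scaled by the side count and floored into sector indices.
# Each index becomes a fraction of a full turn driving a Rotator on offset, doubled UVs.
# Sides clamps to at least one, guarding against division by zero.
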MaterialFunction /FortUILibrary/MaterialFunctions/Basics/MF_UI_UVRadialMirror
graph {
 properties {
  bExposeToLibrary=true
  description=Generates UVs rotationally mirrored around a center point according to the number of sides.
  previewMaterialDomain=MD_UI
  libraryCategoriesText [UI_Library]
 }
 nodes [
  n1 = FunctionInput(Preview=MaterialFunctionCall() {materialFunction=&GetUserInterfaceUV}.Normalized UV) {
   bUsePreviewValueAsDefault=true
   description="Optional custom UV mapping (defaults to Normalized UVs)."
   inputName=UVs
   inputType=FunctionInput_Vector2
  }
  n2 = FunctionInput() {
   bUsePreviewValueAsDefault=true
   description=The number of sides to mirror.
   inputName=Sides
   inputType=FunctionInput_Scalar
   sortPriority=1
   previewValue=<3 0 0 1>
  }
  n3 = Max(A=@n2) {constB=1, desc="Minimum of one side, also prevent divide by zero in next step"}
  n4 = Multiply(
   A=@n3
   B=MaterialFunctionCall!(Vector or UVs=OneMinus(None=@n1)) {materialFunction=&/MF_UI_CheaperVectorToRadial}.Vector Converted to Angle
  ) {constA=1}
  n5 = Subtract(A=Divide(A=Floor(None=@n4), B=Subtract(A=@n3))) {constB=0.5}
  n6 = Subtract(B=Divide(B=@n3))
  n7 = Add(A=@n1, B=Constant2Vector() {r=-0.25})
  FunctionOutput(None=Rotator(Coordinate=Multiply(A=@n7) {constB=2}, Time=Multiply(A=@n6, B=@n5)) {centerY=1, speed=6.2832}) {description=The transformed UVs., outputName=UVs}
 ]
}
