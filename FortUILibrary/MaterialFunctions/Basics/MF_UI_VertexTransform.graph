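# UI material function that applies a render-style transform to widget vertices, with Anchor, Translation, Scale and Padding inputs.
# Anchor and pixel Translation, adjusted by viewport scale, offset the position, while Scale and Padding expand it around the UV center using the widget's pixel size.
# It outputs the transformed position for Screen Position, the new pixel size, and a total scale factor to compensate for the vertex scaling.
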MaterialFunction /FortUILibrary/MaterialFunctions/Basics/MF_UI_VertexTransform
graph {
 properties {bExposeToLibrary=true, description=Apply render transform to the material, previewMaterialDomain=MD_UI, libraryCategoriesText [UI_Library]}
 nodes [
  n1 = FunctionInput() {
   bUsePreviewValueAsDefault=true
   description="Set anchor to move visuals out of bound (-1,1)"
   inputName=Anchor
   inputType=FunctionInput_Vector2
   sortPriority=-1
  }
  n2 = MaterialFunctionCall() {materialFunction=&/MF_UI_ViewportScale}
  n3 = Add(A=WorldPosition().XY, B=Multiply(A=NamedRerouteUsage(), B=Multiply(A=@n1, B=@n2.Viewport Scale Factor)) {constB=-1})
  n4 = FunctionInput() {bUsePreviewValueAsDefault=true, description=Set an offset in pixels, inputName=Translation, inputType=FunctionInput_Vector2}
  n5 = Multiply(A=@n4, B=MaterialFunctionCall() {materialFunction=&MF_UI_ViewportScale}.Viewport Scale Factor)
  n6 = MaterialFunctionCall() {materialFunction=&GetUserInterfaceUV}
  n7 = Multiply(A=NamedRerouteDeclaration(None=@n6.Pixel Size) {name=Pixel Size, nodeColor=<rgba 0 0 0 1>}, B=Subtract(A=@n6.Normalized UV) {constB=0.5})
  n8 = FunctionInput() {
   bUsePreviewValueAsDefault=true
   description=Define the scale
   inputName=Scale
   inputType=FunctionInput_Vector2
   sortPriority=10
   previewValue=<1 1 0 1>
  }
  n9 = NamedRerouteDeclaration(None=@n8) {name=Scale, nodeColor=<rgba 0 0 0 1>}
  n10 = FunctionInput() {bUsePreviewValueAsDefault=true, description=Define the padding in pixels, inputName=Padding, inputType=FunctionInput_Vector2, sortPriority=5}
  n11 = MaterialFunctionCall() {materialFunction=&MF_UI_ViewportScale}
  n12 = Divide(A=NamedRerouteUsage(), B=@n11.Viewport Scale Factor)
  n13 = Multiply(
   A=@n9
   B=Divide(
    A=NamedRerouteDeclaration(None=@n10) {name=Padding, nodeColor=<rgba 0 0 0 1>}
    B=NamedRerouteDeclaration(None=@n12) {name=Scaled Pixel Size, nodeColor=<rgba 0 0 0 1>}
   )
  )
  n14 = Add(A=Subtract(A=@n9), B=@n13)
  FunctionOutput(None=Add(A=Add(A=@n3, B=@n5), B=Multiply(A=@n7, B=@n14))) {description=New vertex position to be sent to Screen Position, outputName=Transformed Position}
  n15 = NamedRerouteUsage()
  n16 = Multiply(A=@n11.Viewport Scale Factor, B=NamedRerouteUsage())
  n17 = NamedRerouteUsage()
  FunctionOutput(None=Multiply(A=@n15, B=Add(A=@n16, B=@n17))) {description=Updated pixel size after vertex transform, outputName=New Pixel Size, sortPriority=5}
  FunctionOutput(None=Add(A=@n13, B=@n15)) {
   description="Provide the scale factor to compensate the vertex scaling (considering both scaling and padding)"
   outputName=Total Scale Factor
   sortPriority=10
  }
 ]
}
