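# WIP UI material function that computes nine-slice UVs, with a hard-coded slice scale and no function inputs.
# GetUserInterfaceUV supplies normalized and pixel-size values; scaled UVs from opposite corners, a pixel-ratio term and their minimum form the corner UVs.
# Floored, saturated per-axis edge tests build corner, edge and center masks without branching.
# The output sums separately scaled corner, edge and center UV terms, linked through named reroutes.
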MaterialFunction /FortUILibrary/MaterialFunctions/Basics/MF_UI_WIP_9SliceUV
graph {nodes [
  n1 = Multiply(A=NamedRerouteUsage(), B=Multiply(A=NamedRerouteUsage()) {constB=0.5}) {desc=Corners UV}
  n2 = Multiply(A=Min(A=NamedRerouteUsage())) {constB=0.5}
  n3 = Multiply(A=NamedRerouteUsage(), B=@n2) {desc=Edges UV}
  n4 = NamedRerouteUsage()
  FunctionOutput(None=Add(A=Add(A=@n1, B=@n3), B=Multiply(A=@n4, B=@n2) {desc=Center UV}))
  n5 = MaterialFunctionCall() {materialFunction=&GetUserInterfaceUV}
  NamedRerouteDeclaration(None=@n5.Pixel Size) {name=Pixel UV, nodeColor=<rgba 0 0 0 1>}
  n6 = Floor(None=NamedRerouteUsage())
  n7 = Floor(None=NamedRerouteUsage())
  n8 = Multiply(A=Saturate(None=ComponentMask(None=@n6) {desc=Left Edge, r=true}), B=Saturate(None=ComponentMask(None=@n7) {desc=Right Edge, r=true}))
  n9 = Multiply(A=Saturate(None=ComponentMask(None=@n6) {desc=Top Edge, g=true}), B=Saturate(None=ComponentMask(None=@n7) {desc=Bottom Edge, g=true}))
  n10 = Saturate(None=Add(A=@n8, B=@n9))
  NamedRerouteDeclaration(None=OneMinus(None=@n10)) {name=Corners Mask, nodeColor=<rgba 0 0 0 1>}
  n11 = Multiply(A=@n8, B=@n9)
  NamedRerouteDeclaration(None=@n11) {name=Center Mask, nodeColor=<rgba 0 0 0 1>}
  NamedRerouteDeclaration(None=Subtract(A=@n10, B=@n11)) {name=Edges Mask, nodeColor=<rgba 0 0 0 1>}
  n12 = NamedRerouteUsage()
  n13 = Constant() {r=3}
  n14 = NamedRerouteUsage()
  n15 = Multiply(A=Multiply(A=@n12, B=@n13) {constB=3}, B=@n14)
  NamedRerouteDeclaration(None=@n15) {name=Top Left UV, nodeColor=<rgba 0 0 0 1>}
  n16 = Multiply(A=Multiply(A=OneMinus(None=@n12), B=@n13) {constB=3}, B=@n14)
  NamedRerouteDeclaration(None=@n16) {name=Bottom Right UV, nodeColor=<rgba 0 0 0 1>}
  NamedRerouteDeclaration(None=Min(A=@n15, B=@n16)) {name=Corners UV, nodeColor=<rgba 0 0 0 1>}
  n17 = MaterialFunctionCall(Float2=NamedRerouteUsage()) {materialFunction=&BreakOutFloat2Components}
  n18 = Divide(A=@n17.R, B=@n17.R)
  n19 = Divide(A=@n17.G, B=@n17.G)
  NamedRerouteDeclaration(None=Max(A=AppendVector(A=@n18, B=@n19))) {name=Max Pixel Ratio, nodeColor=<rgba 0 0 0 1>}
  NamedRerouteDeclaration(None=@n5.Normalized UV) {name=Normalized UV, nodeColor=<rgba 0 0 0 1>}
 ]}
